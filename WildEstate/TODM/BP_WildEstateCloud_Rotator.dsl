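; Wild Estate cloud rotator Blueprint that scatters static mesh clouds in a ring around the actor, with a RotatingMovement component.
; Construction script seeds a random stream, then adds Number of Clouds + 1 meshes at random angles with varied radius, height and uniform scale.
; BeginPlay destroys RotatingMovement on mobile, Switch or PC perf mode, or when the Athena volumetric cloud quality cvar is above zero.
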(event EventBeginPlay
  (bind _returnvalue (Development|GetConsoleVariableIntValue "Athena.VolumetricClouds.Quality"))
  (if (Math|Boolean|ORBoolean (TouchPlatform|IsMobileGame) (TouchPlatform|IsNintendoSwitchorNintendoSwitchPreview) (Class|PlatformHelpersLibrary|IsPcPerfMode) (> _returnvalue 0))
    (Components|DestroyComponent (Variables|Default|GetRotatingMovement))))

(fn ConstructionScript
  (Math|Random|SetRandomStreamSeed (Variables|Default|GetStream) (Variables|Default|GetNew%20Seed))
  (for _ (range (+ (Variables|Default|GetNumber%20of%20Clouds) 1))
    (AddComponent|NODE_AddStaticMeshComponent-0 :RelativeTransform (Math|Transform|MakeTransform (+ (* (Math|Vector|RotateVectorAroundAxis (Math|Vector|MakeVector 1.0) (Math|Random|RandomFloatinRangefromStream (Variables|Default|GetStream) 0.0 3.6e+02) "0.000000,0.000000,1.000000") (Math|Random|RandomFloatinRangefromStream (Variables|Default|GetStream) (- (Variables|Default|GetRadius) (Variables|Default|GetRadius_Variation)) (+ (Variables|Default|GetRadius) (Variables|Default|GetRadius_Variation)))) (Math|Vector|MakeVector 0.0 0.0 (Math|Random|RandomFloatinRangefromStream (Variables|Default|GetStream) (- (Variables|Default|GetHeight) (Variables|Default|GetHeight_Variation)) (+ (Variables|Default|GetHeight) (Variables|Default|GetHeight_Variation))))) "0, 0, 0" (Math|Conversions|ToVector(Float) (Math|Random|RandomFloatinRangefromStream (Variables|Default|GetStream) (- (Variables|Default|GetScale) (Variables|Default|GetScale_Variation)) (+ (Variables|Default|GetScale) (Variables|Default|GetScale_Variation))))))))
