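# Unlit translucent sky material that draws either the sun or a textured moon with a halo, blended by Moon Color alpha.
# World position offset places and sizes the quad far along the offset sky atmosphere light direction, using cross products to face it along that axis.
# UV distance builds the disc mask and halo falloff for color and opacity.
# Camera vector clips it below the horizon, and pass switches zero it in reflection captures.
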Material /WildEstate/TODM/M_SundAndMoon_WildEstate
graph {
 properties {
  bAllowNegativeEmissiveColor=true
  bAutomaticallySetUsageInEditor=false
  bUseTranslucencyVertexFog=false
  blendMode=BLEND_Translucent
  opacityMaskClipValue=0.5
  shadingModel=MSM_Unlit
  translucencyPass=MTP_BeforeDOF
  usedShadingModels=Unlit
 }
 nodes [
  ComponentMask() {g=true, r=true}
  n1 = Add(A=TextureCoordinate()) {constB=-0.5}
  n2 = Distance(A=Multiply(A=@n1) {constB=2}, B=Constant())
  n3 = OneMinus(None=Saturate(None=Multiply(A=@n2) {constB=3}))
  NamedRerouteDeclaration(None=Saturate(None=Multiply(A=@n3) {constB=30})) {name=Circle Mask, nodeColor=<rgba 0.3125 1 0 1>}
  MaterialExpressionAdd_3 = Add(A=Multiply(A=@n1) {constB=2.9}) {constB=0.5}
  n4 = Convert(None=@MaterialExpressionAdd_3) {
   nodeName=Break Float2
   convertInputs [{
     type=Vector2
     expressionInput {
      expression=&self:MaterialExpressionAdd_3
      inputName=None
      mask=0
      maskA=0
      maskB=0
      maskG=0
      maskR=0
      outputIndex=0
     }
    }]
   convertOutputs [{type=Vector2}]
  }
  NamedRerouteDeclaration(None=@n4) {name=Moon UVs, nodeColor=<rgba 1 0 0.6328 1>}
  n5 = Transform(None=LocalPosition().XYZ)
  n6 = MaterialFunctionCall() {materialFunction=&ObjectScale}
  n7 = SkyAtmosphereLightDirection()
  n8 = VectorParameter() {parameterName=SunVectorOffset, defaultValue=<rgba 0 0 0 1>}
  n9 = MaterialFunctionCall(R=@n8.R, G=Multiply(A=ComponentMask(None=@n7) {r=true}, B=@n8.G), B=@n8.R) {materialFunction=&AppendMany}
  n10 = Add(A=Normalize(VectorInput=Add(A=@n7, B=@n9.RGB)), B=AppendVector(A=Constant2Vector(), B=ScalarParameter() {parameterName=Horizon Offset}))
  n11 = Add(A=Multiply(A=Multiply(A=@n5, B=@n6.Scale XYZ)) {constB=-1}, B=Multiply(A=@n10) {constB=3000000})
  n12 = LocalPosition()
  n13 = ComponentMask(None=@n12.XY) {g=true}
  n14 = Normalize(VectorInput=CrossProduct(A=Constant3Vector() {constant=<rgba 0 0 1 0>}, B=@n10))
  n15 = ComponentMask(None=@n12.XY) {r=true}
  n16 = CrossProduct(A=@n14, B=@n10)
  n17 = Multiply(A=Add(A=Multiply(A=@n13, B=@n14), B=Multiply(A=@n15, B=@n16)), B=ScalarParameter() {defaultValue=3500, parameterName=Size}) {constB=60}
  NamedRerouteDeclaration(None=Subtract(A=Add(A=@n11, B=@n17), B=ActorPositionWS() {originType=CameraRelative})) {name=WPO, nodeColor=<rgba 0 1 0.7188 1>}
  n18 = OneMinus(None=Saturate(None=@n2))
  n19 = Multiply(
   A=Power(Base=Multiply(A=@n18) {constB=1.5}, Exp=ScalarParameter() {defaultValue=4, parameterName=Halo Exponent}) {constExponent=8}
   B=NamedRerouteUsage()
  )
  n20 = NamedRerouteUsage()
  n21 = CameraVectorWS()
  n22 = Saturate(None=Multiply(A=ComponentMask(None=@n21) {b=true}) {constB=-1})
  NamedRerouteDeclaration(None=Multiply(A=Max(A=@n19, B=@n20), B=Ceil(None=@n22))) {name=Opacity, nodeColor=<rgba 1 0 0.5625 1>}
  n23 = VectorParameter() {parameterName=Halo Color, defaultValue=<rgba 10 4.4549 0 1>}
  n24 = Multiply(A=VectorParameter() {parameterName=Sun Color, defaultValue=<rgba 1 0.2482 0 1>}.RGB) {constB=500}
  n25 = TextureSampleParameter2D(UVs=NamedRerouteUsage(), Apply View MipBias=Constant()) {
   group=Sun-Moon
   parameterName=MoonTexture
   samplerSource=SSM_Clamp_WorldGroupSettings
   texture=&/STK_COL_Moon_Texture_WildEstate
  }
  n26 = VectorParameter() {parameterName=Moon Color, defaultValue=<rgba 1 1 1 0>}
  n27 = Multiply(A=@n25.RGB, B=Multiply(A=@n26.RGB) {constB=2})
  NamedRerouteDeclaration(None=LinearInterpolate(A=@n23.RGB, B=LinearInterpolate(A=@n24, B=@n27, Alpha=@n26.A), Alpha=NamedRerouteUsage())) {name=Color, nodeColor=<rgba 0 0.8828 1 1>}
  NamedRerouteDeclaration(None=@n23.A) {name=Halo Alpha, nodeColor=<rgba 0.4062 1 0 1>}
  NamedRerouteUsage()
  ReflectionCapturePassSwitch!(Default=NamedRerouteUsage(), Reflection=Constant())
  ReflectionCapturePassSwitch(Default=NamedRerouteUsage(), Reflection=Constant())
 ]
}
